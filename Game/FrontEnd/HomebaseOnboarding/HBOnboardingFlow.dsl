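; Homebase onboarding flow Blueprint for the front end that decides whether to play the periodic event movie; its StartFlow event is empty.
; A PlayEventMovie client event checks the Campaign quest, played flag and non-mobile platform, then spawns the movie announcement if onboarding cinematics are enabled.
; Cinematic client events report the outcome, and completion records a quest stat.
; A one-shot 0.1-second timer retries through duplicated handler logic, an admitted hack for late quest grants.
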(event Custom|StartFlow)

(event ClientEvent|HandleClientEvent:PlayEventMovie
  (params (EventSource :type Object%20Reference) (EventFocus :type Object%20Reference) (ClientEvent :type Fort%20Client%20Event%20Structure%20%28by%20ref%29))
  ;[1 1 1] Play any event movie(s), post onboarding but before continuing.
  (Utilities|IsValid (GlobalUIContext|GetQuestManager (BlueprintContext|GetFortGlobalUIContext) "Campaign")
    (:Is%20Valid
      (if (CallFunction|NeedstoPlayEventMovie)
        (if (Utilities|Enum|Equal(Enum) (GlobalUIContext|GetFeatureState (BlueprintContext|GetFortGlobalUIContext) "ShowOnboardingCinematics") "Enabled")
          (Utilities|MissionMacroSpawnAnnouncement :AnnouncementToSpawn (Variables|Default|GetPeriodicEventMovieAnnouncementClass) :WaitForClients true
            (:Then
              _)
            (:ClientsDone
              (CallFunction|ClientEvent|ExecuteClientEvent "(CategoryName=\"Cinematic\",EventName=\"EventMovieFinishedPlaying\")") ; No matter what (unlress we are waiting to check one more time),\ntell FrontEnd we have finished playing once we have decided \n(whether we have actually played or just chosen not to ever play).
              (Quests|SendClientStatEvent (CallFunction|GetCampaignQuestManager) (Variables|QuestData|GetPlayPeriodicEventCineObjective)))) ; Play Periodic Event Movie
          (else
            (Utilities|FlowControl|Sequence
              (:then_0
                _)
              (:then_1
                _))))
        (Variables|QuestData|SetbPlayedPeriodicEventMovie true)
        (else
          (bind _waitingtocheckagain (CallFunction|RecheckNeedtoPlayEventMovie))
          (if _waitingtocheckagain
            (Development|PrintString "Waiting to recheck whether we need to play the event movie, but just very briefly in case MCP info isn't back.  Will decide soon." false)
            (else
              (Development|PrintString "Event Movie Quest is not in progress.  Either the quest has not yet been granted or it has already been completed (i.e. the movie has already played).  In either case, we will not play it." false))))))
    (:Is%20Not%20Valid
      (bind _waitingtocheckagain_2 (CallFunction|RecheckNeedtoPlayEventMovie))
      (if _waitingtocheckagain_2
        (Development|PrintString "Quest manager not found.  Will recheck once before giving up on playing event movies." false)
        (else
          (Development|PrintString "Failed to find Quest Manager.  Cannot determine if event quests need to be played, so will play none." false)))))
  (CallFunction|ClientEvent|ExecuteClientEvent "(CategoryName=\"Cinematic\",EventName=\"NotPlayingEventMovie\")") ; No matter what (unlress we are waiting to check one more time),\ntell FrontEnd we have finished playing once we have decided \n(whether we have actually played or just chosen not to ever play).
)

(event Custom|HandlePlayEventCine
  ;[1 1 1] Play any event movie(s), post onboarding but before continuing.
  (Utilities|IsValid (GlobalUIContext|GetQuestManager (BlueprintContext|GetFortGlobalUIContext) "Campaign")
    (:Is%20Valid
      (if (CallFunction|NeedstoPlayEventMovie)
        (if (Utilities|Enum|Equal(Enum) (GlobalUIContext|GetFeatureState (BlueprintContext|GetFortGlobalUIContext) "ShowOnboardingCinematics") "Enabled")
          (Utilities|MissionMacroSpawnAnnouncement :AnnouncementToSpawn (Variables|Default|GetPeriodicEventMovieAnnouncementClass) :WaitForClients true
            (:Then
              _)
            (:ClientsDone
              (CallFunction|ClientEvent|ExecuteClientEvent "(CategoryName=\"Cinematic\",EventName=\"EventMovieFinishedPlaying\")") ; No matter what (unlress we are waiting to check one more time),\ntell FrontEnd we have finished playing once we have decided \n(whether we have actually played or just chosen not to ever play).
              (Quests|SendClientStatEvent (CallFunction|GetCampaignQuestManager) (Variables|QuestData|GetPlayPeriodicEventCineObjective)))) ; Play Periodic Event Movie
          (else
            (Utilities|FlowControl|Sequence
              (:then_0
                _)
              (:then_1
                _))))
        (Variables|QuestData|SetbPlayedPeriodicEventMovie true)
        (else
          (bind _waitingtocheckagain (CallFunction|RecheckNeedtoPlayEventMovie))
          (if _waitingtocheckagain
            (Development|PrintString "Waiting to recheck whether we need to play the event movie, but just very briefly in case MCP info isn't back.  Will decide soon." false)
            (else
              (Development|PrintString "Event Movie Quest is not in progress.  Either the quest has not yet been granted or it has already been completed (i.e. the movie has already played).  In either case, we will not play it." false))))))
    (:Is%20Not%20Valid
      (bind _waitingtocheckagain_2 (CallFunction|RecheckNeedtoPlayEventMovie))
      (if _waitingtocheckagain_2
        (Development|PrintString "Quest manager not found.  Will recheck once before giving up on playing event movies." false)
        (else
          (Development|PrintString "Failed to find Quest Manager.  Cannot determine if event quests need to be played, so will play none." false)))))
  (CallFunction|ClientEvent|ExecuteClientEvent "(CategoryName=\"Cinematic\",EventName=\"NotPlayingEventMovie\")") ; No matter what (unlress we are waiting to check one more time),\ntell FrontEnd we have finished playing once we have decided \n(whether we have actually played or just chosen not to ever play).
)

(fn GetCampaignQuestManager
  (return (GlobalUIContext|GetQuestManager (BlueprintContext|GetFortGlobalUIContext) "Campaign")))

(fn NeedsToPlayEventMovie
  (return (Math|Boolean|ANDBoolean (not (Variables|QuestData|GetbPlayedPeriodicEventMovie)) (Quest|IsQuestinProgress (GlobalUIContext|GetQuestManager (BlueprintContext|GetFortGlobalUIContext) "Campaign") (Variables|QuestData|GetPlayPeriodicEventMovieQuest)) (not (GlobalUIContext|IsMobileApp (BlueprintContext|GetFortGlobalUIContext))))))

(fn RecheckNeedToPlayEventMovie
  (if (and (not (Variables|QuestData|GetbPlayedPeriodicEventMovie)) (not (Variables|QuestData|GetbHasRecheckedNeedToPlayPeriodicEventMovie)))
    (Variables|QuestData|SetbHasRecheckedNeedToPlayPeriodicEventMovie true)
    ;[1 0 0 0.2] s Evil lives
    (Utilities|Time|SetTimerbyEvent (EventDispatchers|CreateEvent :!event HandlePlayEventCine) 0.1) ; This is a horrible hack!  Setting a timer to wait\nuntil we are given the quest before rechecking\nto see if we need to play the movie.  Apparently\nwe grant the quest AFTER we check in the FE, \nso we need to try one more time before giving\nup.
    (return true)
    (else
      (return false))))

(fn ConstructionScript)
